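; AnimNotify_OliveStomp_InvHullOutlines is an animation notify for the OliveStomp emote that applies an inverted-hull outline material to the animated skeletal mesh.
; Its Received_NotifyTick sets the mesh component's overlay material to the OliveStomp InvHullOutline material instance on every tick of the notify window.
; The file contains no begin or end handler, so nothing here clears the overlay afterward.
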(fn Received_NotifyTick
  (params (MeshComp :type Skeletal%20Mesh%20Component%20Object%20Reference) (Animation :type Anim%20Sequence%20Base%20Object%20Reference) (FrameDeltaTime :type Float%20%28single-precision%29) (EventReference :type Anim%20Notify%20Event%20Reference%20Structure%20%28by%20ref%29))
  (Rendering|Material|SetOverlayMaterial MeshComp "/BRCosmetics/Animation/Game/MainPlayer/Emotes/Olivestomp/MI_Emote_OliveStomp_InvHullOutline.MI_Emote_OliveStomp_InvHullOutline" false 0 :!class /Script/Engine.MeshComponent)
  (return false))
