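# Niagara system for the OliveStomp emote that plays a short sheen on a mesh, finished by a sparkle.
# One burst emitter drives both mesh and sprite renderers, animating a dynamic material parameter from 0 to 1 over particle life.
# FN_Sprite_Attributes writes separate sprite position, spin and size, so the sparkle pops near the end at a system-defined offset.
# A system-level TODColor, apparently time-of-day based, scales the particle color.
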NiagaraSystem /BRCosmetics/Animation/Game/MainPlayer/Emotes/Olivestomp/NS_Emote_OliveStomp_Sheen
dependencies {
 usedDataInterfaces [&NiagaraDataInterfaceCurve, &NiagaraDataInterfaceColorCurve]
 usedDynamicInputs [
  &Lerp_Float
  &UniformRangedFloat
  &FNTimeOfDayScalar
  &ColorFromCurve
 ]
 usedModules [
  &self:SystemSpawnScript.SystemScriptSource.SystemScriptGraph.NiagaraNodeAssignment_0.SetVariables_#1
  &SystemState
  &self:SystemSpawnScript.SystemScriptSource.SystemScriptGraph.NiagaraNodeAssignment_1.SetVariables_#2
  &self:SystemSpawnScript.SystemScriptSource.SystemScriptGraph.NiagaraNodeAssignment_2.SetVariables_#3
  &self:SystemSpawnScript.SystemScriptSource.SystemScriptGraph.NiagaraNodeAssignment_3.SetVariables_#4
 ]
}
emitters {Sheen_Mesh {
  emitterUpdateScript [
   &EmitterState()
   &SpawnBurst_Instantaneous(Use Spawn Probability=false, UseLoopCountLimit=false)
  ]
  particleSpawnScript [
   &InitializeParticle(
    Color=<rgba 0.3 0.3 0.3 1>
    Lifetime=0.8
    Mesh Scale Mode=::NewEnumerator3
    Position Offset Coordinate Space=::Local
    UsePositionOffset=true
    Write Material Random=true
   )
   &DynamicMaterialParameters(
    Index 0 Param 2=0.6
    Module.Param0WriteEnabled=false
    Module.Param1WriteEnabled=true
    Module.Param2WriteEnabled=true
    Module.Param3WriteEnabled=false
   )
   &CameraOffset() {enabled=false}
  ]
  particleUpdateScript [
   &ParticleState()
   &DynamicMaterialParameters(
    Index 0 Param 1=&FloatFromCurve(FloatCurve=DataInterface() {propertyValues {Curve {keys [{interpMode=RCIM_Cubic, tangentMode=RCTM_User, time=0, value=0}, {interpMode=RCIM_Cubic, tangentMode=RCTM_User, time=1, value=1}]}}})
    Module.Param0WriteEnabled=true
    Module.Param1WriteEnabled=false
    Module.Param2WriteEnabled=true
    Module.Param3WriteEnabled=false
   ) {module=DynamicMaterialParameters001}
   &ScaleColor(Scale Alpha=&Multiply_Float(A=1, B=$System.Timeline.InverseNormalizedTime), ScaleA=true, ScaleRGB=true) {enabled=false}
   &ScaleColor(
    Scale Mode=::NewEnumerator1
    Scale RGBA=&MakeVector4_FromLinearColor(LinearColor=$System.TODColor)
    ScaleRGBA=true
   ) {module=ScaleColor001}
   &FN_Sprite_Attributes(
    Create Position Attribute=true
    Create Sprite Rotation Attribute=true
    Create SpriteSize Attribute=true
    Position:NiagaraPosition=$System.SparklePositionOffset
    SpriteRotation:NiagaraFloat=&Multiply_Float(A=&RandomRangeFloat(Minimum=100, Maximum=200), B=$Emitter.Age)
    SpriteSize=&Vector2DFromFloat(Value=&FloatFromCurve(
      FloatCurve=DataInterface() {propertyValues {Curve {keys [
          {interpMode=RCIM_Cubic, tangentMode=RCTM_User, time=0.7, value=0}
          {interpMode=RCIM_Cubic, tangentMode=RCTM_User, time=0.8, value=1}
          {interpMode=RCIM_Cubic, tangentMode=RCTM_User, time=1, value=0}
         ]}}}
      Scale Curve=70
     ))
    Write to Module Namespace=true
   )
  ]
  rendererClasses [&NiagaraMeshRendererProperties, &NiagaraSpriteRendererProperties]
  renderers [
   {rendererClass=&NiagaraMeshRendererProperties, rendererIndex=0}
   {rendererClass=&NiagaraSpriteRendererProperties, rendererIndex=1}
  ]
 }}
summary {
 systemName=NS_Emote_OliveStomp_Sheen
 emitters [{emitterName=Sheen_Mesh, rendererClasses [&NiagaraMeshRendererProperties, &NiagaraSpriteRendererProperties]}]
 userVariables [
  User.Debug {type=NiagaraFloat, defaultValue=0, description=""}
  User.SkeletalMesh {
   type=NiagaraDataInterfaceSkeletalMesh
   defaultValue {
    struct=NiagaraExt_VariableValue_DataInterface
    value {dataInterface=&self:NiagaraDataInterfaceSkeletalMesh_0}
   }
  }
 ]
}
